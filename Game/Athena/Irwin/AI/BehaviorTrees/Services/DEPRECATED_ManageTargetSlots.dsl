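; Deprecated behavior tree service for Irwin AI that tracks how many AI occupy attack-behavior slots on their GoalActor, using Gameplay Effect stacks.
; Activation applies a counting effect to the target, a max effect once the stack limit is reached, and a slot-owner effect to the pawn.
; Deactivation removes the counting effect and, unless the pawn has an override-slot effect, the owner effect and any no-longer-valid max effect.
; OnDied only serves debug printing.
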(event EventReceiveActivation
  (params (OwnerActor :type Actor%20Object%20Reference))
  (bind _returnvalue (AI|GetBlackboard OwnerActor))
  ;[1 1 1] Track attack behavior slots on our Target.
  ;
  ; Count how many AI are slotted into attack behaviors of a particular type.
  ;
  ; Allows AI to slot-in or hold back and wait if the slots are currently filled.
  (Utilities|IsValid OwnerActor
    (:Is%20Valid
      (bind _asfort_aicontroller (Utilities|Casting|CastToFortAIController OwnerActor))
      (bind _myfortpawn (Variables|/Script/FortniteAI.FortAIController|GetMyFortPawn _asfort_aicontroller)
        (:then
          (bind _output_get (Variables|Default|SetActivatingPawn _myfortpawn :!pinned true)) ; Save Activating Pawn
          (Utilities|IsValid (Ability|GetAbilitySystemComponent _output_get)
            (:Is%20Valid
              (Variables|Default|SetActivatingPawnAbilitySystemComponent (Ability|GetAbilitySystemComponent _output_get) :!pinned true) ; Save Activating Pawn Ability System
              (Utilities|IsValid _returnvalue
                (:Is%20Valid
                  (bind _asactor (Utilities|Casting|CastToActor (AI|Components|Blackboard|GetValueasObject _returnvalue (Utilities|Name|MakeLiteralName "GoalActor"))))
                  (Utilities|IsValid (Ability|GetAbilitySystemComponent _asactor)
                    (:Is%20Valid
                      (Variables|Default|SetTargetAbilitySystemComponent (Ability|GetAbilitySystemComponent _asactor) :!pinned true) ; Save Target Ability System
                      (Utilities|FlowControl|Sequence
                        (:then_0
                          (bind _targetabilitysystemcomponent (Variables|Default|GetTargetAbilitySystemComponent)
                            (:then
                              (bind _returnvalue_2 (GameplayEffects|ApplyGameplayEffectToTarget :self _targetabilitysystemcomponent :GameplayEffectClass (Variables|Default|GetSlotCountingGameplayEffect) :Target _targetabilitysystemcomponent :Level 0.0 :!pinned true)) ; Apply Counting GE to Target
                              (Variables|Default|SetActiveAppliedSlotCountingGameplayEffect _returnvalue_2))
                            (:else)))
                        (:then_1
                          (bind (_slotcount _slotcountmax _slotcountismax__or_more _slotcountismax__exactly) (CallFunction|GetSlotCountTotals))
                          (if _slotcountismax__or_more
                            (GameplayEffects|ApplyGameplayEffectToTarget :self (Variables|Default|GetTargetAbilitySystemComponent) :GameplayEffectClass (Variables|Default|GetSlotCountingMaxGameplayEffect) :Target (Variables|Default|GetTargetAbilitySystemComponent) :Level 0.0 :!pinned true) ; Apply Max GE to Target
                            (if _slotcountismax__exactly
                              _)
                            (else
                              _))
                          (bind _activatingpawnabilitysystemcomponent (Variables|Default|GetActivatingPawnAbilitySystemComponent)
                            (:then
                              (bind _returnvalue_3 (GameplayEffects|ApplyGameplayEffectToTarget :self _activatingpawnabilitysystemcomponent :GameplayEffectClass (Variables|Default|GetSlotOwnerGameplayEffect) :Target _activatingpawnabilitysystemcomponent :Level 0.0 :!pinned true)) ; Apply SlotOwner GE to Activating Pawn
                              (Variables|Default|SetActiveAppliedSlotOwnerGameplayEffect _returnvalue_3))
                            (:else)))
                        (:then_2
                          (Development|PrintString (Utilities|String|BuildString(Integer) (Utilities|String|BuildString(Integer) "===== STACK ADD " "(" _slotcount "/") "" _slotcountmax ") =====") (Variables|Default|GetDebug) (Variables|Default|GetDebug)))
                        (:then_3
                          (if (Variables|Default|GetDebug)
                            (bind _activatingpawn (Variables|Default|GetActivatingPawn)
                              (:then
                                (EventDispatchers|AssignOnDied _activatingpawn (delegate Custom|OnDied)))
                              (:else))))))))))))
        (:else)))))

(event Custom|OnDied
  (params (DamagedActor :type Actor%20Object%20Reference) (Damage :type Float%20%28single-precision%29) (InstigatedBy :type Controller%20Object%20Reference) (DamageCauser :type Actor%20Object%20Reference) (HitLocation :type Vector) (FHitComponent :type Primitive%20Component%20Object%20Reference) (BoneName :type Name) (Momentum :type Vector))
  ;[1 1 1] Track attack behavior slots on our Target.
  ;
  ; Count how many AI are slotted into attack behaviors of a particular type.
  ;
  ; Allows AI to slot-in or hold back and wait if the slots are currently filled.
  (bind _activatingpawn (Variables|Default|GetActivatingPawn)
    (:then
      (EventDispatchers|UnbindEventfromOnDied _activatingpawn (delegate Custom|OnDied))
      (bind (_slotcount _slotcountmax) (CallFunction|GetSlotCountTotals))
      (Development|PrintString (Utilities|String|BuildString(Integer) (Utilities|String|BuildString(Integer) "===== STACK DAID " "(" (- _slotcount 1) "/") "" _slotcountmax ") =====") (Variables|Default|GetDebug) (Variables|Default|GetDebug)))
    (:else)))

(event EventReceiveDeactivation
  (params (OwnerActor :type Actor%20Object%20Reference))
  ;[1 1 1] Track attack behavior slots on our Target.
  ;
  ; Count how many AI are slotted into attack behaviors of a particular type.
  ;
  ; Allows AI to slot-in or hold back and wait if the slots are currently filled.
  (Utilities|FlowControl|Sequence
    (:then_0
      (bind _targetabilitysystemcomponent (Variables|Default|GetTargetAbilitySystemComponent)
        (:then
          (GameplayEffects|RemoveActiveGameplayEffect _targetabilitysystemcomponent (Variables|Default|GetActiveAppliedSlotCountingGameplayEffect) 1 :!pinned true) ; Remove Counting GE from Target
        )
        (:else)))
    (:then_1
      (bind _override (CallFunction|ActivatingPawnIsUsingOverrideSlot))
      (if (not _override)
        (bind _activatingpawnabilitysystemcomponent (Variables|Default|GetActivatingPawnAbilitySystemComponent)
          (:then
            (GameplayEffects|RemoveActiveGameplayEffect _activatingpawnabilitysystemcomponent (Variables|Default|GetActiveAppliedSlotOwnerGameplayEffect) 1 :!pinned true) ; Remove SlotOwner GE from Activating Pawn
            (bind (_slotcount _slotcountmax _slotcountismax__or_more) (CallFunction|GetSlotCountTotals))
            (if (not _slotcountismax__or_more)
              (GameplayEffects|RemoveActiveGameplayEffectbySourceEffect (Variables|Default|GetTargetAbilitySystemComponent) (Variables|Default|GetSlotCountingMaxGameplayEffect) (Variables|Default|GetTargetAbilitySystemComponent) 1 :!pinned true) ; Remove Max GE from Target
            ))
          (:else))))
    (:then_2
      (if (Variables|Default|GetDebug)
        (bind _activatingpawn (Variables|Default|GetActivatingPawn)
          (:then
            (EventDispatchers|UnbindEventfromOnDied _activatingpawn (delegate Custom|OnDied)))
          (:else))))))

(event EventReceiveTick
  (params (OwnerActor :type Actor%20Object%20Reference) (DeltaSeconds :type Float%20%28single-precision%29)))

(fn ActivatingPawnIsUsingOverrideSlot
  (bind _finaleffectcount (Class|AbilityFunctions|GetTotalGameplayEffectCount (Utilities|Array|MakeArray (Variables|Default|GetOverrideSlotGameplayEffect)) (Variables|Default|GetActivatingPawnAbilitySystemComponent)))
  (return (!= _finaleffectcount 0)))

(fn GetSlotCountTotals
  (bind (_period _bexecuteperiodiceffectonapplication _periodicinhibitionpolicy _modifiers _executions _conditionalgameplayeffects _overfloweffects _bdenyoverflowapplication _bclearstackonoverflow _prematureexpirationeffectclasses _routineexpirationeffectclasses _brequiremodifiersuccesstotriggercues _gameplaycues _inheritablegameplayeffecttags _inheritableownedtagscontainer _inheritableblockedabilitytagscontainer _ongoingtagrequirements _applicationtagrequirements _removaltagrequirements _removegameplayeffectswithtags _grantedapplicationimmunitytags _grantedapplicationimmunityquery _removegameplayeffectquery _stackingtype _stacklimitcount) (Class|GetClassDefaults (Variables|Default|GetSlotCountingGameplayEffectStackLimit)))
  (bind _targetabilitysystemcomponent (Variables|Default|GetTargetAbilitySystemComponent)
    (:then
      (bind _finaleffectcount (Class|AbilityFunctions|GetTotalGameplayEffectCount (Utilities|Array|MakeArray (Variables|Default|GetSlotCountingGameplayEffect)) _targetabilitysystemcomponent))
      (return _finaleffectcount _stacklimitcount (>= _finaleffectcount _stacklimitcount) (== _finaleffectcount _stacklimitcount)))
    (:else
      (return -1 -1 false false))))
